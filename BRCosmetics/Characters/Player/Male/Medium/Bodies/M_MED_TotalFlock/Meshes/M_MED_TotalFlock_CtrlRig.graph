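# Control Rig for the Total Flock medium male player outfit that procedurally poses its front and back cloth, skirt and hip straps from leg movement.
# A shared DeformRig function handles body deformation, while spherical pose readers on the thighs produce up and back alphas.
# These alphas, swing-twist leg angles and thigh rotation weight additive bone offsets on the cloth and straps.
# Sliding, running or emote, and rail-grind states blend in smoothed adjustments.
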ControlRigBlueprint /BRCosmetics/Characters/Player/Male/Medium/Bodies/M_MED_TotalFlock/Meshes/M_MED_TotalFlock_CtrlRig
graphs [
 Add_8_ContainedGraph {nodes [
   n1 = RigVMFunctionEntryNode()
   n2 = &RigVMFunction_MathDoubleAdd(A=@n1.A, B=@n1.B)
   RigVMFunctionReturnNode(Result=&RigVMFunction_MathDoubleAdd(A=&RigVMFunction_MathDoubleAdd(A=@n2.Result, B=@n1.C).Result, B=@n1.D).Result)
  ]}
 CollapseNode_ContainedGraph {nodes [
   RigVMFunctionEntryNode()
   > RigUnit_SetTransform = &RigUnit_SetTransform(
    Item=(Type=Bone,Name="thigh_r")
    Weight=1
    bInitial=False
    Value=&RigUnit_GetTransform(Item=(Type=Control,Name="thigh_r_ctrl"), bInitial=False).Transform
   )
   > &RigUnit_SetTransform(
    Item=(Type=Bone,Name="thigh_l")
    Weight=1
    bInitial=False
    Value=&RigUnit_GetTransform(Item=(Type=Control,Name="thigh_l_ctrl"), bInitial=False).Transform
   )
   RigUnit_GetTransform = &RigUnit_GetTransform(Item=(Type=Control,Name="pelvis_ctrl"), bInitial=False)
   > &RigUnit_SetTransform(Item=(Type=Bone,Name="pelvis"), Weight=1, bInitial=False, Value=@RigUnit_GetTransform.Transform)
   > &RigUnit_SetTransform(
    Item=(Type=Bone,Name="spine_01")
    Weight=1
    bInitial=False
    Value=&RigUnit_GetTransform(Item=(Type=Control,Name="spine_01_ctrl"), bInitial=False).Transform
   )
   > &RigUnit_SetTransform(
    Item=(Type=Bone,Name="spine_02")
    Weight=1
    bInitial=False
    Value=&RigUnit_GetTransform(Item=(Type=Control,Name="spine_02_ctrl"), bInitial=False).Transform
   )
   > &RigUnit_SetTransform(
    Item=(Type=Bone,Name="spine_03")
    Weight=1
    bInitial=False
    Value=&RigUnit_GetTransform(Item=(Type=Control,Name="spine_03_ctrl"), bInitial=False).Transform
   )
   > &RigUnit_SetTransform(
    Item=(Type=Bone,Name="spine_04")
    Weight=1
    bInitial=False
    Value=&RigUnit_GetTransform(Item=(Type=Control,Name="spine_04_ctrl"), bInitial=False).Transform
   )
   > RigVMFunctionReturnNode()
   RigVMRerouteNode(Value=())
   RigVMRerouteNode(Value=())
  ]}
 CollapseNode_ContainedGraph {nodes [
   RigVMFunctionEntryNode()
   > RigVMFunction_Sequence = RigVMAggregateNode()
   n1 = &RigUnit_GetRelativeTransformForItem(Child=(Type=Bone,Name="thigh_twist_01_r"), Parent=(Type=Control,Name="ref_thigh_twist_01_r_ctrl"), bChildInitial=False, bParentInitial=False)
   RigVMFunction_MathQuaternionSwingTwist = &RigVMFunction_MathQuaternionSwingTwist(TwistAxis=(X=1,Y=0,Z=0), Input=@n1.RelativeTransform.Rotation)
   RigVMFunction_MathQuaternionToEuler = &RigVMFunction_MathQuaternionToEuler(RotationOrder=ZYX, Value=@RigVMFunction_MathQuaternionSwingTwist.Swing)
   RigVMVariableNode(Value=@RigVMFunction_MathQuaternionToEuler.Result, ExecuteContext=@RigVMFunction_Sequence.B) {variable=Leg_r_Swing_rot}
   n2 = &RigUnit_GetRelativeTransformForItem(Child=(Type=Bone,Name="thigh_twist_01_l"), Parent=(Type=Control,Name="ref_thigh_twist_01_l_ctrl"), bChildInitial=False, bParentInitial=False)
   n3 = &RigVMFunction_MathQuaternionToEuler(
    RotationOrder=ZYX
    Value=&RigVMFunction_MathQuaternionSwingTwist(TwistAxis=(X=1,Y=0,Z=0), Input=@n2.RelativeTransform.Rotation).Swing
   )
   > RigVMVariableNode(Value=@n3.Result) {variable=Leg_l_Swing_rot}
   > RigVMVariableNode(Value=/Script/RigVM.RigVMFunction_MathDoubleMin(A=@RigVMFunction_MathQuaternionToEuler.Result.Z, B=@n3.Result.Z) {is=struct RigVMUnitNode, template="Minimum::Execute(in A,in B,out Result)"}.Result) {variable=leg_rot_z_minimum}
   > RigVMVariableNode(Value=&RigVMFunction_MathDoubleAdd(A=@RigVMFunction_MathQuaternionToEuler.Result.Y, B=@n3.Result.Y).Result) {variable=leg_rot_y_add}
   Remap = &RigVMFunction_MathDoubleRemap(SourceMaximum=-100, SourceMinimum=0, TargetMaximum=-15, TargetMinimum=0, Value=RigVMVariableNode() {variable=leg_rot_z_minimum}.Value)
   n4 = RigVMVariableNode() {variable=leg_rot_y_add}
   n5 = RigVMVariableNode() {variable=Leg_r_Swing_rot}
   n6 = RigVMVariableNode() {variable=Leg_l_Swing_rot}
   RigVMFunction_AnimEvalRichCurve = &RigVMFunction_AnimEvalRichCurve(
    Curve="(EditorCurveData=(Keys=((InterpMode=RCIM_Linear,TangentMode=RCTM_Auto,TangentWeightMode=RCTWM_WeightedNone,Time=0,Value=0,ArriveTangent=1.1332,ArriveTangentWeight=0,LeaveTangent=0,LeaveTangentWeight=0),(InterpMode=RCIM_Linear,TangentMode=RCTM_Auto,TangentW ...(truncated)"
    SourceMaximum=-158
    SourceMinimum=0
    TargetMinimum=0
    TargetMaximum=&RigVMFunction_MathDoubleMul(B=-160, A=RigVMVariableNode() {variable=grindReduction}.Value).Result
    Value=RigVMVariableNode() {variable=leg_rot_z_minimum}.Value
   )
   RigVMFunction_MathQuaternionFromEuler = &RigVMFunction_MathQuaternionFromEuler(
    Euler=(X=0,Y=30,Z=0)
    Euler.Y=&RigVMFunction_MathDoubleAdd(
     A=&RigVMFunction_MathDoubleRemap(SourceMaximum=-140, SourceMinimum=0, TargetMaximum=-40, TargetMinimum=0, Value=@n5.Value.Z).Result
     B=&RigVMFunction_MathDoubleRemap(SourceMaximum=-140, SourceMinimum=0, TargetMaximum=40, TargetMinimum=0, Value=@n6.Value.Z).Result
    ).Result
    Euler.Z=@RigVMFunction_AnimEvalRichCurve.Result
   )
   &RigUnit_ModifyTransforms(
    ItemToModify=((Item=(Type=Bone,Name="dyn_frontMiddleCloth_00"),Transform=(Rotation=(X=0,Y=0,Z=0,W=1),Translation=(X=0,Y=0,Z=0),Scale3D=(X=1,Y=1,Z=1))))
    Weight=1
    WeightMinimum=0
    ItemToModify.0.Transform.Translation.X=&RigVMFunction_MathDoubleMin(
     A=@Remap.Result
     B=&RigVMFunction_MathDoubleRemap(SourceMaximum=-70, SourceMinimum=0, TargetMaximum=-15, TargetMinimum=0, Value=@n4.Value).Result
    ) {template="Minimum::Execute(in A,in B,out Result)"}.Result
    ItemToModify.0.Transform.Rotation=@RigVMFunction_MathQuaternionFromEuler.Result
    ExecutePin=@RigVMFunction_Sequence.C
   )
   RigVMFunctionReturnNode(ExecuteContext=@RigVMFunction_Sequence.F)
   n7 = &RigVMFunction_MathDoubleRemap(SourceMaximum=-140, SourceMinimum=0, TargetMaximum=-22, TargetMinimum=0, Value=RigVMVariableNode() {variable=Leg_r_Swing_rot}.Value.Z)
   n8 = &RigVMFunction_MathDoubleRemap(SourceMaximum=-140, SourceMinimum=0, TargetMaximum=22, TargetMinimum=0, Value=RigVMVariableNode() {variable=Leg_l_Swing_rot}.Value.Z)
   n9 = &RigVMFunction_AlphaInterp(
    InterpSpeedDecreasing=10
    InterpSpeedIncreasing=10
    Scale=1
    bClampResult=False
    bMapRange=false
    Value=RigVMDispatchNode(False=-140, True=-160, Condition=RigVMVariableNode() {variable=isRunningOrEmote}.Value) {template="DISPATCH_RigVMDispatch_If(in Condition,in True,in False,out Result)"}.Result
   )
   n10 = &RigVMFunction_AnimEvalRichCurve(
    Curve="(EditorCurveData=(Keys=((InterpMode=RCIM_Linear,TangentMode=RCTM_Auto,TangentWeightMode=RCTWM_WeightedNone,Time=0,Value=0,ArriveTangent=1.1332,ArriveTangentWeight=0,LeaveTangent=0,LeaveTangentWeight=0),(InterpMode=RCIM_Linear,TangentMode=RCTM_Auto,TangentW ...(truncated)"
    SourceMaximum=-150
    SourceMinimum=0
    TargetMinimum=0
    TargetMaximum=&RigVMFunction_MathDoubleMul(B=@n9.Result, A=RigVMVariableNode() {variable=grindReduction}.Value).Result
    Value=RigVMVariableNode() {variable=leg_rot_z_minimum}.Value
   )
   &RigUnit_ModifyTransforms(
    ItemToModify=((Item=(Type=Bone,Name="base_frontCloth"),Transform=(Rotation=(X=0,Y=0,Z=0,W=1),Translation=(X=0,Y=0,Z=0),Scale3D=(X=1,Y=1,Z=1))))
    Weight=1
    WeightMinimum=0
    ItemToModify.0.Transform.Rotation=&RigVMFunction_MathQuaternionFromEuler(Euler=(X=0,Y=30,Z=0), Euler.Y=&RigVMFunction_MathDoubleAdd(A=@n7.Result, B=@n8.Result).Result, Euler.Z=@n10.Result).Result
    ExecutePin=@RigVMFunction_Sequence.E
   )
   n11 = RigVMVariableNode() {variable=l_thigh_up_alpha}
   > &RigUnit_ModifyTransforms(
    ItemToModify="((Item=(Type=Bone,Name=\"base_frontCloth_l\"),Transform=(Rotation=(X=0,Y=0,Z=0,W=1),Translation=(X=0,Y=-2.5,Z=0),Scale3D=(X=1,Y=1,Z=1))),(Item=(Type=Bone,Name=\"base_frontCloth_r\"),Transform=(Rotation=(X=0.0262,Y=-0.0831,Z=0.2996,W=0.9501),Translation=(X=0,Y= ...(truncated)"
    WeightMinimum=0
    Weight=@n11.Value
   )
   n12 = RigVMVariableNode() {variable=r_thigh_up_alpha}
   > &RigUnit_ModifyTransforms(
    ItemToModify="((Item=(Type=Bone,Name=\"base_frontCloth_l\"),Transform=(Rotation=(X=-0.0522,Y=0.1656,Z=0.2961,W=0.9392),Translation=(X=0,Y=0,Z=0),Scale3D=(X=1,Y=1,Z=1))),(Item=(Type=Bone,Name=\"base_frontCloth_r\"),Transform=(Rotation=(X=0,Y=0,Z=0,W=1),Translation=(X=0,Y=-2. ...(truncated)"
    WeightMinimum=0
    Weight=@n12.Value
   )
   Add = &RigVMFunction_MathDoubleAdd(A=@n11.Value, B=@n12.Value)
   > &RigUnit_ModifyTransforms(
    ItemToModify="((Item=(Type=Bone,Name=\"base_frontCloth_l\"),Transform=(Rotation=(X=0,Y=0,Z=-0.1305,W=0.9914),Translation=(X=2,Y=-2,Z=0),Scale3D=(X=1,Y=1,Z=1))),(Item=(Type=Bone,Name=\"base_frontCloth_r\"),Transform=(Rotation=(X=0,Y=0,Z=-0.1305,W=0.9914),Translation=(X=2,Y=- ...(truncated)"
    WeightMinimum=0
    Weight=&RigVMFunction_MathDoubleMul(B=0.5, A=@Add.Result).Result
   )
   n13 = RigVMVariableNode() {variable=Leg_r_Swing_rot}
   n14 = /Script/RigVM.RigVMFunction_MathVectorRemap(
    SourceMaximum=(X=-65,Y=-65,Z=-65)
    SourceMinimum=(X=0,Y=0,Z=0)
    TargetMaximum=(X=2,Y=2,Z=1)
    TargetMinimum=(X=0,Y=0,Z=0)
    Value=(X=0,Y=0,Z=0)
    bClamp=true
    Value.Y=@n13.Value.Y
    Value.Z=@n13.Value.Y
    Value.X=RigVMVariableNode() {variable=Leg_l_Swing_rot}.Value.Y
   ) {
    is=struct RigVMUnitNode
    template="Remap::Execute(in Value,in SourceMinimum,in SourceMaximum,in TargetMinimum,in TargetMaximum,in bClamp,out Result)"
   }
   Multiply = /Script/RigVM.RigVMFunction_MathVectorMul(A=(X=0,Y=-50,Z=-25), B=(X=1,Y=1,Z=1), B.X=@n14.Result.Z, B.Y=@n14.Result.Z, B.Z=@n14.Result.Z) {is=struct RigVMUnitNode, template="Multiply::Execute(in A,in B,out Result)"}
   n15 = &RigVMFunction_MathQuaternionFromEuler(Euler=@Multiply.Result)
   &RigUnit_ModifyTransforms(
    ItemToModify="((Item=(Type=Bone,Name=\"base_backCloth_l\"),Transform=(Rotation=(X=0,Y=0,Z=0,W=1),Translation=(X=0,Y=2,Z=0),Scale3D=(X=1,Y=1,Z=1))),(Item=(Type=Bone,Name=\"base_backCloth_r\"),Transform=(Rotation=(X=0,Y=0,Z=0,W=1),Translation=(X=0,Y=2,Z=0),Scale3D=(X=1,Y=1,Z= ...(truncated)"
    Weight=1
    WeightMinimum=0
    ItemToModify.0.Transform.Translation.Y=@n14.Result.X
    ItemToModify.1.Transform.Translation.Y=@n14.Result.Y
    ItemToModify.2.Transform.Rotation=@n15.Result
    ItemToModify.3.Transform.Rotation=@n15.Result
    ExecutePin=@RigVMFunction_Sequence.D
   )
   AlphaInterp = &RigVMFunction_AlphaInterp(
    InterpSpeedDecreasing=10
    InterpSpeedIncreasing=10
    Scale=1
    bClampResult=False
    bMapRange=false
    Value=RigVMDispatchNode(False=0, True=-6, Condition=RigVMVariableNode() {variable=isRunningOrEmote}.Value) {template="DISPATCH_RigVMDispatch_If(in Condition,in True,in False,out Result)"}.Result
   )
   n16 = RigVMVariableNode() {variable=leg_rot_z_minimum}
   n17 = RigVMVariableNode() {variable=leg_rot_y_add}
   &RigVMFunction_MathDoubleAdd(
    B=@AlphaInterp.Result
    A=&RigVMFunction_MathDoubleMin(
     A=&RigVMFunction_MathDoubleRemap(SourceMaximum=-100, SourceMinimum=0, TargetMaximum=-3, TargetMinimum=0, Value=@n16.Value).Result
     B=&RigVMFunction_MathDoubleRemap(SourceMaximum=-70, SourceMinimum=0, TargetMaximum=-3, TargetMinimum=0, Value=@n17.Value).Result
    ) {template="Minimum::Execute(in A,in B,out Result)"}.Result
   )
   &RigUnit_ModifyTransforms(
    ExecutePin=()
    ItemToModify=((Item=(Type=Bone,Name="dyn_frontMiddleCloth_00"),Transform=(Rotation=(X=0,Y=0,Z=-0.766,W=0.6428),Translation=(X=-10,Y=0,Z=0),Scale3D=(X=1,Y=1,Z=1))))
    WeightMinimum=0
    Weight=&RigVMFunction_MathDoubleMax(A=RigVMVariableNode() {variable=l_thigh_up_alpha}.Value, B=RigVMVariableNode() {variable=r_thigh_up_alpha}.Value).Result
   )
  ]}
 CollapseNode_ContainedGraph {nodes [
   RigVMFunctionEntryNode()
   AlphaInterp = &RigVMFunction_AlphaInterp(
    InterpSpeedDecreasing=10
    InterpSpeedIncreasing=10
    Scale=1
    bClampResult=true
    Value=RigVMDispatchNode(False=0, True=1, Condition=RigVMVariableNode() {variable=Sliding}.Value) {template="DISPATCH_RigVMDispatch_If(in Condition,in True,in False,out Result)"}.Result
   )
   > &RigUnit_ModifyTransforms(
    ItemToModify="((Item=(Type=Bone,Name=\"base_backCloth_r\"),Transform=(Rotation=(X=0,Y=0,Z=0.5,W=0.866),Translation=(X=0,Y=5,Z=0),Scale3D=(X=1,Y=1,Z=1))),(Item=(Type=Bone,Name=\"base_backCloth_l\"),Transform=(Rotation=(X=0,Y=0,Z=0.5,W=0.866),Translation=(X=0,Y=5,Z=0),Scale3D ...(truncated)"
    WeightMinimum=0
    Weight=@AlphaInterp.Result
   )
   > &RigUnit_ModifyTransforms(
    ItemToModify=((Item=(Type=Bone,Name="ref_foot_l"),Transform=(Rotation=(X=0,Y=0,Z=0,W=1),Translation=(X=0,Y=9,Z=0),Scale3D=(X=1,Y=1,Z=1))),(Item=(Type=Bone,Name="ref_foot_r"),Transform=(Rotation=(X=0,Y=0,Z=0,W=1),Translation=(X=0,Y=-9,Z=0),Scale3D=(X=1,Y=1,Z=1))))
    WeightMinimum=0
    Weight=&RigVMFunction_MathDoubleSub(A=1, B=@AlphaInterp.Result).Result
   )
   > RigVMFunctionReturnNode()
  ]}
 New Function_ContainedGraph {nodes [
   n1 = RigVMFunctionEntryNode()
   RigVMFunction_MathQuaternionToEuler = &RigVMFunction_MathQuaternionToEuler(
    RotationOrder=XYZ
    Value=&RigUnit_GetRelativeTransformForItem(
     Child=(Type=Bone,Name="hand_l")
     Parent=(Type=Null,Name="lowerarm_l")
     bChildInitial=False
     bParentInitial=False
     Child.Name=@n1.Child
     Parent.Name=@n1.Parent
    ).RelativeTransform.Rotation
   )
   Remap = &RigVMFunction_MathDoubleRemap(
    TargetMaximum=1
    TargetMinimum=0
    Value=@RigVMFunction_MathQuaternionToEuler.Result.Y
    SourceMaximum=@n1.Source Maximum
    SourceMinimum=@n1.Source Minimum
   )
   > RigVMFunctionReturnNode(Delta Rotation=@Remap.Result)
  ]}
 RigVMFunctionLibrary {nodes [Delta Rotation Y = RigVMCollapseNode(ExecuteContext=())]}
 RigVMFunction_Sequence_1_ContainedGraph {nodes [
   n1 = RigVMFunctionEntryNode()
   RigVMFunction_Sequence = &RigVMFunction_Sequence(ExecuteContext=@n1.ExecuteContext)
   n2 = &RigVMFunction_Sequence(ExecuteContext=@RigVMFunction_Sequence.B)
   n3 = &RigVMFunction_Sequence(ExecuteContext=@n2.B)
   n4 = &RigVMFunction_Sequence(ExecuteContext=@n3.B)
   n5 = &RigVMFunction_Sequence(ExecuteContext=@n4.B)
   n6 = &RigVMFunction_Sequence(ExecuteContext=@n5.B)
   RigVMFunctionReturnNode(A=@RigVMFunction_Sequence.A, B=@n2.A, C=@n3.A, D=@n4.A, E=@n5.A, F=@n6.A, G=@n6.B)
  ]}
 RigVMFunction_Sequence_1_ContainedGraph {nodes [
   n1 = RigVMFunctionEntryNode()
   RigVMFunction_Sequence = &RigVMFunction_Sequence(ExecuteContext=@n1.ExecuteContext)
   n2 = &RigVMFunction_Sequence(ExecuteContext=@RigVMFunction_Sequence.B)
   n3 = &RigVMFunction_Sequence(ExecuteContext=@n2.B)
   n4 = &RigVMFunction_Sequence(ExecuteContext=@n3.B)
   n5 = &RigVMFunction_Sequence(ExecuteContext=@n4.B)
   RigVMFunctionReturnNode(A=@RigVMFunction_Sequence.A, B=@n2.A, C=@n3.A, D=@n4.A, E=@n5.A, F=@n5.B)
  ]}
 RigVMModel {nodes [
   n1 = RigVMVariableNode() {variable=l_thigh_back_alpha}
   n2 = RigVMVariableNode() {variable=r_thigh_back_alpha}
   Subtract = &RigVMFunction_MathDoubleSub(A=@n1.Value, B=@n2.Value)
   Remap = &RigVMFunction_MathDoubleRemap(SourceMaximum=1, SourceMinimum=0, TargetMaximum=1, TargetMinimum=0.1, Value=@Subtract.Result)
   &RigUnit_ParentConstraint(
    AdvancedSettings=(InterpolationType=Average,RotationOrderForFilter=XZY)
    Child=(Type=Bone,Name="base_backMiddleCloth_00")
    ExecutePin=()
    Filter=(TranslationFilter=(bX=true,bY=true,bZ=true),RotationFilter=(bX=true,bY=true,bZ=true),ScaleFilter=(bX=True,bY=True,bZ=True))
    Parents=((Item=(Type=Bone,Name="thigh_l"),Weight=0.25),(Item=(Type=Bone,Name="thigh_r"),Weight=1))
    Weight=1
    Parents.0.Weight=@Remap.Result
    Parents.1.Weight=&RigVMFunction_MathDoubleRemap(SourceMaximum=0, SourceMinimum=-1, TargetMaximum=0.1, TargetMinimum=1, Value=@Subtract.Result).Result
   )
   Add_10 = RigVMAggregateNode(
    B=&RigVMFunction_MathDoubleMul(A=60, B=RigVMVariableNode() {variable=r_thigh_up_alpha}.Value).Result
    D=&RigVMFunction_MathDoubleMul(A=-60, B=RigVMVariableNode() {variable=r_thigh_back_alpha}.Value).Result
    A=&RigVMFunction_MathDoubleMul(A=-60, B=RigVMVariableNode() {variable=l_thigh_up_alpha}.Value).Result
    C=&RigVMFunction_MathDoubleMul(A=60, B=RigVMVariableNode() {variable=l_thigh_back_alpha}.Value).Result
   )
   n3 = &RigVMFunction_MathQuaternionFromEuler(Euler=(X=0,Y=0,Z=0), RotationOrder=XYZ, Euler.X=@Add_10.Result)
   > &RigUnit_ModifyTransforms(
    ItemToModify="((Item=(Type=Bone,Name=\"base_frontMiddleCloth_00\"),Transform=(Rotation=(X=-0.3827,Y=0,Z=0,W=0.9239),Translation=(X=0,Y=0,Z=0),Scale3D=(X=1,Y=1,Z=1))),(Item=(Type=Bone,Name=\"dyn_backMiddleCloth_00\"),Transform=(Rotation=(X=0.3827,Y=0,Z=0,W=0.9239),Translatio ...(truncated)"
    Weight=1
    WeightMinimum=0
    ItemToModify.0.Transform.Rotation=@n3.Result
    ItemToModify.1.Transform.Rotation=@n3.Result
   )
   RigUnit_BeginExecution = &RigUnit_BeginExecution()
   RigVMFunction_Sequence = RigVMAggregateNode(ExecuteContext=@RigUnit_BeginExecution.ExecutePin)
   FN_DeformRig_FNC = RigVMFunctionReferenceNode(
    DeformRig_ Curve="(deformRig_l_upperarm_up_alpha_72_#1=False,deformRig_r_upperarm_up_alpha_73_#2=False,deformRig_l_upperarm_fr_alpha_74_#3=False,deformRig_r_upperarm_fr_alpha_75_CB51B8 ...(truncated)"
    DeformRig_Bool="(upperarm_1_#4=True,elbow_3_#5=True,wrist_7_#6=True,groin_8_#7=True,glute_17_#8=True,knee_9_40C408324CA261 ...(truncated)"
    DeformRig_Debug=(upperarm_debug_18_#4=False,wrist_debug_20_#6=False,lat_debug_26_#9=False,groinGlute_debug_27_#7=False)
    DeformRig_Value="(upperarm_up_18_#10=(Rotation=(X=0.0,Y=-0.0,Z=0.0,W=1.0),Translation=(X=0.0,Y=0.0,Z=8.0),Scale3D=(X=0.70,Y=1.0,Z=1.0)),upperarm_fwd_fr_19_#11=(Rotation=(X=0.0,Y=-0.0,Z=0.0,W=1.0),Translation=(X=0.0, ...(truncated)"
    ExecuteContext=@RigVMFunction_Sequence.B
   )
   n4 = &RigUnit_SphericalPoseReader(
    ActiveRegionScaleFactors=(PositiveWidth=3.5,NegativeWidth=0.5,PositiveHeight=4.2,NegativeHeight=3)
    ActiveRegionSize=0.1
    Debug=(bDrawDebug=false,DebugScale=25,DebugSegments=20,DebugThickness=0.25)
    DriverAxis=(X=-1,Y=0,Z=0)
    DriverItem=(Type=Bone,Name="thigh_l")
    FalloffRegionScaleFactors=(PositiveWidth=1,NegativeWidth=1.11,PositiveHeight=4,NegativeHeight=1)
    FalloffSize=0.45
    OptionalParentItem=(Type=Bone)
    RotationOffset=(X=90,Y=0,Z=0)
    ExecutePin=@RigVMFunction_Sequence.C
   )
   > RigVMVariableNode(Value=@n4.OutputParam) {variable=l_thigh_up_alpha}
   > n5 = &RigUnit_SphericalPoseReader(
    ActiveRegionScaleFactors=(PositiveWidth=3.5,NegativeWidth=0.5,PositiveHeight=4.2,NegativeHeight=3)
    ActiveRegionSize=0.1
    Debug=(bDrawDebug=false,DebugScale=25,DebugSegments=20,DebugThickness=0.25)
    DriverAxis=(X=1,Y=0,Z=0)
    DriverItem=(Type=Bone,Name="thigh_r")
    FalloffRegionScaleFactors=(PositiveWidth=1,NegativeWidth=1.11,PositiveHeight=4,NegativeHeight=1)
    FalloffSize=0.45
    FlipHeightScaling=false
    FlipWidthScaling=true
    OptionalParentItem=(Type=Bone)
    RotationOffset=(X=-90,Y=0,Z=0)
   )
   > RigVMVariableNode(Value=@n5.OutputParam) {variable=r_thigh_up_alpha}
   > n6 = &RigUnit_SphericalPoseReader(
    ActiveRegionScaleFactors=(PositiveWidth=3.5,NegativeWidth=0.5,PositiveHeight=3,NegativeHeight=3)
    ActiveRegionSize=0.1
    Debug=(bDrawDebug=false,bDraw2D=false,DebugScale=25,DebugSegments=20,DebugThickness=0.25)
    DriverAxis=(X=-1,Y=0,Z=0)
    DriverItem=(Type=Bone,Name="thigh_l")
    FalloffRegionScaleFactors=(PositiveWidth=1,NegativeWidth=1.11,PositiveHeight=1,NegativeHeight=1)
    FalloffSize=0.45
    OptionalParentItem=(Type=Bone)
    RotationOffset=(X=-90,Y=0,Z=0)
   )
   > RigVMVariableNode(Value=@n6.OutputParam) {variable=l_thigh_back_alpha}
   > n7 = &RigUnit_SphericalPoseReader(
    ActiveRegionScaleFactors=(PositiveWidth=3.5,NegativeWidth=0.5,PositiveHeight=3,NegativeHeight=3)
    ActiveRegionSize=0.1
    Debug=(bDrawDebug=false,DebugScale=25,DebugSegments=20,DebugThickness=0.25)
    DriverAxis=(X=1,Y=0,Z=0)
    DriverItem=(Type=Bone,Name="thigh_r")
    FalloffRegionScaleFactors=(PositiveWidth=1,NegativeWidth=1.11,PositiveHeight=1,NegativeHeight=1)
    FalloffSize=0.45
    FlipHeightScaling=false
    FlipWidthScaling=true
    OptionalParentItem=(Type=Bone)
    RotationOffset=(X=90,Y=0,Z=0)
   )
   > RigVMVariableNode(Value=@n7.OutputParam) {variable=r_thigh_back_alpha}
   AlphaInterp = &RigVMFunction_AlphaInterp(
    InterpSpeedDecreasing=10
    InterpSpeedIncreasing=10
    Scale=1
    bClampResult=true
    Value=RigVMDispatchNode(False=1, True=0.6, Condition=RigVMVariableNode() {variable=grindOnRail}.Value) {template="DISPATCH_RigVMDispatch_If(in Condition,in True,in False,out Result)"}.Result
   )
   > RigVMVariableNode(Value=@AlphaInterp.Result) {variable=grindReduction}
   n8 = RigVMVariableNode() {variable=l_thigh_back_alpha}
   n9 = RigVMVariableNode() {variable=r_thigh_back_alpha}
   Maximum = &RigVMFunction_MathDoubleMax(B=@n8.Value, A=@n9.Value)
   &RigUnit_ModifyTransforms(
    ItemToModify="((Item=(Type=Bone,Name=\"base_backCloth_r\"),Transform=(Rotation=(X=0,Y=0,Z=0.7373,W=0.6756),Translation=(X=0,Y=0,Z=0),Scale3D=(X=1,Y=1,Z=1))),(Item=(Type=Bone,Name=\"base_backCloth_l\"),Transform=(Rotation=(X=0,Y=0,Z=0.7373,W=0.6756),Translation=(X=0,Y=0,Z=0) ...(truncated)"
    WeightMinimum=0
    Weight=@Maximum.Result
    ExecutePin=@RigVMFunction_Sequence.E
   )
   > &RigUnit_ModifyTransforms(
    ItemToModify=((Item=(Type=Bone,Name="dyn_backMiddleCloth_00"),Transform=(Rotation=(X=0,Y=0,Z=0.7934,W=0.6088),Translation=(X=-20,Y=0,Z=0),Scale3D=(X=1,Y=1,Z=1))))
    WeightMinimum=0
    Weight=@Maximum.Result
   )
   > Sliding Setup = RigVMCollapseNode()
   > Delta Rotation Y = RigVMFunctionReferenceNode(Child=thigh_r, Parent=thigh_r_null, Source Maximum=-80)
   > &RigUnit_ModifyTransforms(
    ItemToModify="((Item=(Type=Bone,Name=\"base_strap_A_r\"),Transform=(Rotation=(X=-0.018,Y=-0.2079,Z=0.102,W=0.9726),Translation=(X=0,Y=0,Z=0),Scale3D=(X=1,Y=1,Z=1))),(Item=(Type=Bone,Name=\"base_strap_B_r\"),Transform=(Rotation=(X=0.1421,Y=-0.1663,Z=0.1067,W=0.9699),Translat ...(truncated)"
    WeightMinimum=0
    Weight=@"Delta Rotation Y".Delta Rotation
   )
   Front White Skirt Setup = RigVMCollapseNode(ExecuteContext=@RigVMFunction_Sequence.D)
   n10 = &RigUnit_SphericalPoseReader(
    ActiveRegionScaleFactors=(PositiveWidth=1,NegativeWidth=2,PositiveHeight=1.5,NegativeHeight=1.5)
    ActiveRegionSize=0.2
    Debug=(bDrawDebug=true,bDraw2D=False,bDrawLocalAxes=False,DebugScale=25,DebugSegments=20,DebugThickness=0.25)
    DriverAxis=(X=1,Y=0,Z=0)
    DriverItem=(Type=Bone,Name="spine_05")
    ExecutePin=()
    FalloffRegionScaleFactors=(PositiveWidth=2,NegativeWidth=0,PositiveHeight=2,NegativeHeight=2)
    FalloffSize=0.4
    FlipHeightScaling=False
    FlipWidthScaling=False
    OptionalParentItem=(Type=Bone,Name="spine_01")
    RotationOffset=(X=90,Y=0,Z=40)
   )
   > RigVMVariableNode() {variable=grindOnRail}
   > RigVMVariableNode(Value=&RigVMFunction_MathDoubleSub(B=@n10.OutputParam, A=RigVMVariableNode() {variable=l_thigh_up_alpha}.Value).Result) {variable=l_thigh_up_alpha}
   > RigVMVariableNode(Value=&RigVMFunction_MathDoubleSub(B=@n10.OutputParam, A=RigVMVariableNode() {variable=r_thigh_up_alpha}.Value).Result) {variable=r_thigh_up_alpha}
   n11 = &RigVMFunction_MathDoubleAdd(A=1, B=&RigVMFunction_MathDoubleMul(B=0.15, A=RigVMVariableNode() {variable=GetCrouchingAlpha}.Value).Result)
   n12 = RigVMVariableNode() {variable=r_thigh_up_alpha}
   n13 = &RigVMFunction_MathDoubleMul(A=15, B=@n12.Value)
   n14 = RigVMVariableNode() {variable=l_thigh_up_alpha}
   n15 = &RigVMFunction_MathDoubleMul(A=-25, B=@n14.Value)
   n16 = &RigVMFunction_MathDoubleMul(A=-30, B=@n14.Value)
   n17 = &RigVMFunction_MathDoubleMul(A=30, B=@n12.Value)
   RigVMVariableNode(
    ExecuteContext=()
    Value=(Rotation=(X=0,Y=0,Z=0,W=1),Translation=(X=0,Y=0,Z=0),Scale3D=(X=0.9,Y=0.9,Z=0.9))
    Value.Translation=&RigVMFunction_MathVectorMul(A=(X=-6,Y=-3,Z=0), B=(X=1,Y=1,Z=1), B.X=@n11.Result, B.Y=@n11.Result, B.Z=@n11.Result) {template="Multiply::Execute(in A,in B,out Result)"}.Result
    Value.Rotation=&RigVMFunction_MathQuaternionFromEuler(
     Euler=(X=0,Y=0,Z=-80)
     RotationOrder=XYZ
     Euler.X=&RigVMFunction_MathDoubleAdd(B=@n13.Result, A=@n15.Result).Result
     Euler.Y=&RigVMFunction_MathDoubleAdd(B=@n16.Result, A=@n17.Result).Result
     Euler.Z=&RigVMFunction_MathDoubleMul(A=-60, B=@n11.Result).Result
    ).Result
   ) {variable=SetLoinclothFront}
   n18 = RigVMVariableNode() {variable=r_thigh_up_alpha}
   > &RigUnit_ModifyTransforms(
    ItemToModify=((Item=(Type=Bone,Name="base_frontCloth"),Transform=(Rotation=(X=0,Y=0,Z=-0.6428,W=0.766),Translation=(X=-1.4565,Y=-0.4092,Z=0),Scale3D=(X=1,Y=1,Z=1))))
    WeightMinimum=0
    Weight=&RigVMFunction_MathDoubleMax(A=RigVMVariableNode() {variable=l_thigh_up_alpha}.Value, B=@n18.Value).Result
    ItemToModify.0.Transform=RigVMVariableNode() {variable=SetLoinclothFront}.Value
   )
   > &RigUnit_ModifyTransforms(
    ItemToModify="((Item=(Type=Bone,Name=\"None\"),Transform=(Rotation=(X=0,Y=0,Z=0,W=1),Translation=(X=0,Y=0,Z=0),Scale3D=(X=1,Y=1,Z=1))),(Item=(Type=Bone,Name=\"base_strap_A_r\"),Transform=(Rotation=(X=-0.0519,Y=-0.2775,Z=-0.2125,W=0.9355),Translation=(X=0,Y=0,Z=0),Scale3D=(X ...(truncated)"
    WeightMinimum=0
    Weight=@n18.Value
   )
   Add = &RigVMFunction_MathDoubleAdd(B=@n8.Value, A=@n9.Value)
   Controls = RigVMCollapseNode(ExecuteContext=())
   n19 = &RigVMFunction_MathDoubleAdd(A=RigVMVariableNode() {variable=l_thigh_up_alpha}.Value, B=RigVMVariableNode() {variable=r_thigh_up_alpha}.Value)
   Greater = /Script/RigVM.RigVMFunction_MathDoubleGreater(B=0, A=@n19.Result) {is=struct RigVMUnitNode, template="Greater::Execute(in A,in B,out Result)"}
   n20 = &RigVMFunction_MathDoubleAdd(A=@n1.Value, B=@n2.Value)
   RigVMDispatchNode(
    False=1
    Condition=@Greater.Result
    True=RigVMDispatchNode(
     True=1
     False=&RigVMFunction_MathDoubleDiv(B=2, A=@n19.Result).Result
     Condition=&RigVMFunction_MathDoubleGreater(B=0, A=@n20.Result) {template="Greater::Execute(in A,in B,out Result)"}.Result
    ) {template="DISPATCH_RigVMDispatch_If(in Condition,in True,in False,out Result)"}.Result
   ) {template="DISPATCH_RigVMDispatch_If(in Condition,in True,in False,out Result)"}
   &RigUnit_ParentConstraint(
    AdvancedSettings=(InterpolationType=Average,RotationOrderForFilter=XZY)
    Child=(Type=Bone,Name="base_backMiddleCloth_00")
    ExecutePin=()
    Parents=((Item=(Type=Bone,Name="deform_glute_l"),Weight=1),(Item=(Type=Bone,Name="deform_glute_r"),Weight=1))
    Weight=1
   )
   Multiply = &RigVMFunction_MathDoubleMul(B=0.4, A=@n8.Value)
   &RigVMFunction_MathDoubleSub(B=@Multiply.Result, A=@n9.Value)
   &RigVMFunction_MathDoubleSub(B=&RigVMFunction_MathDoubleMul(B=0.4, A=@n9.Value).Result, A=@n8.Value)
   &RigVMFunction_MathDoubleSub(A=1, B=@n10.OutputParam)
  ]}
]
hierarchy [
 root {loc=<0 0 0>, rot=<0 0 0 1>}
 pelvis {parent=root, loc=<0 -0.7071 93.1191>, rot=<0.0127 0.707 -0.0127 -0.707>}
 spine_01 {parent=pelvis, loc=<6.7494 -0.6388 0>, rot=<0 0 0.034 -0.9994>}
 spine_02 {parent=spine_01, loc=<7.2518 -0.035 0>, rot=<0 0 0.0238 -0.9997>}
 spine_03 {parent=spine_02, loc=<6.0104 0 0>, rot=<0 0 0.0149 -0.9999>}
 spine_04 {parent=spine_03, loc=<8.1023 0 0>, rot=<0 0 -0.0132 -0.9999>}
 spine_05 {parent=spine_04, loc=<10.0146 0 0>, rot=<0 0 -0.0015 -1>}
 attach_backpack {parent=spine_05, loc=<-0.1183 -13.4737 -0.0002>, rot=<-0.0093 -0.7071 0.0093 -0.707>}
 neck_01 {parent=spine_05, loc=<11.3138 0.1529 0>, rot=<0 0 -0.2003 -0.9797>}
 neck_02 {parent=neck_01, loc=<4.5627 0 0>, rot=<0 0 0.0504 -0.9987>}
 head {parent=neck_02, loc=<5.3914 0 0>, rot=<0 0 0.115 -0.9934>}
 clavicle_l {parent=spine_05, loc=<4.3575 -0.1466 -2.3223>, rot=<0.0465 -0.7289 -0.009 -0.6829>}
 upperarm_l {parent=clavicle_l, loc=<15.5111 0 0>, rot=<0.0226 -0.3234 -0.0088 -0.9459>}
 lowerarm_l {parent=upperarm_l, loc=<26.6514 0 0>, rot=<0.0391 -0.0099 -0.2453 -0.9686>}
 hand_l {parent=lowerarm_l, loc=<22.3491 0 0>, rot=<-0.7081 -0.0507 0.0975 -0.6975>}
 thumb_01_l {parent=hand_l, loc=<1.3118 -0.3191 -2.459>, rot=<0.6546 -0.4042 0.0664 -0.6354>}
 thumb_02_l {parent=thumb_01_l, loc=<5.9336 -0.0448 0.0401>, rot=<-0.0037 0.0213 0.298 -0.9543>}
 thumb_03_l {parent=thumb_02_l, loc=<2.2223 0.3682 -0.0907>, rot=<0.0118 -0.013 0.0066 -0.9998>}
 index_metacarpal_l {parent=hand_l, loc=<1.6535 0.0183 -2.96>, rot=<0.0158 -0.0124 -0.0173 -0.9996>}
 index_01_l {parent=index_metacarpal_l, loc=<8.3008 0.0357 -0.0524>, rot=<-0.0044 0.0485 0.0924 -0.9945>}
 index_02_l {parent=index_01_l, loc=<3.3587 -0.3004 0.0452>, rot=<-0.002 0.0027 0.3225 -0.9466>}
 index_03_l {parent=index_02_l, loc=<2.5578 -0.123 -0.0252>, rot=<-0.002 0.0067 0.2558 -0.9667>}
 middle_metacarpal_l {parent=hand_l, loc=<2.1856 -0.1725 -0.5984>, rot=<0.006 0.0168 -0.0329 -0.9993>}
 middle_01_l {parent=middle_metacarpal_l, loc=<7.3275 -0.2447 0.0129>, rot=<-0.0208 0.0042 0.0858 -0.9961>}
 middle_02_l {parent=middle_01_l, loc=<4.0123 0.1401 0.1175>, rot=<-0.0218 0.0258 0.3512 -0.9357>}
 middle_03_l {parent=middle_02_l, loc=<2.6077 0.0534 0.1373>, rot=<0.0082 0.0188 0.1708 -0.9851>}
 ring_metacarpal_l {parent=hand_l, loc=<2.2074 -0.3625 1.3638>, rot=<-0.0284 0.0722 -0.0458 -0.9959>}
 ring_01_l {parent=ring_metacarpal_l, loc=<7.0295 -0.6197 -0.1446>, rot=<-0.0489 -0.0252 0.1892 -0.9804>}
 ring_02_l {parent=ring_01_l, loc=<3.6174 0.4647 -0.1027>, rot=<0 0 0.2806 -0.9598>}
 ring_03_l {parent=ring_02_l, loc=<2.4152 0.114 0.0199>, rot=<0 0 0.1189 -0.9929>}
 pinky_metacarpal_l {parent=hand_l, loc=<1.7616 -0.9753 3.3814>, rot=<-0.0397 0.0814 -0.069 -0.9935>}
 pinky_01_l {parent=pinky_metacarpal_l, loc=<7.1709 -0.4831 -0.1094>, rot=<-0.0654 -0.0035 0.2242 -0.9723>}
 pinky_02_l {parent=pinky_01_l, loc=<2.8274 0.0893 -0.1043>, rot=<0.0037 0.0011 0.2854 -0.9584>}
 pinky_03_l {parent=pinky_02_l, loc=<2.2046 0.0307 -0.0078>, rot=<-0.0002 0.0098 0.1445 -0.9895>}
 deform_wrist_dn_base_l {parent=hand_l, loc=<0 -2.2342 0>, rot=<0 0 0 -1>}
 deform_wrist_dn_l {parent=deform_wrist_dn_base_l, loc=<0 0 0>, rot=<0 0 0 1>}
 deform_wrist_up_base_l {parent=hand_l, loc=<0 2.2682 0>, rot=<0 0 0 -1>}
 deform_wrist_up_l {parent=deform_wrist_up_base_l, loc=<0 0 0>, rot=<0 0 0 1>}
 lowerarm_twist_02_l {parent=lowerarm_l, loc=<9.6199 0 0>, rot=<0 0 0 -1>}
 lowerarm_twist_01_l {parent=lowerarm_l, loc=<17.8549 0 0>, rot=<0 0 0 -1>}
 deform_elbow_in_base_l {parent=lowerarm_l, loc=<0.2112 5.259 0.8274>, rot=<0 0 0 -1>}
 deform_elbow_in_l {parent=deform_elbow_in_base_l, loc=<0 0 0>, rot=<0 0 0 1>}
 deform_elbow_out_base_l {parent=lowerarm_l, loc=<0 -4 0>, rot=<0 0 0 -1>}
 deform_elbow_out_l {parent=deform_elbow_out_base_l, loc=<0 0 0>, rot=<0 0 0 1>}
 upperarm_twist_01_l {parent=upperarm_l, loc=<0.0041 0 0>, rot=<0 0 0 -1>}
 deform_upperarm_bk_base_l {parent=upperarm_twist_01_l, loc=<1.0813 -7.2193 -1.9453>, rot=<0 0 0 1>}
 deform_upperarm_bk_l {parent=deform_upperarm_bk_base_l, loc=<0 0 0>, rot=<0 0 0 1>}
 deform_upperarm_fr_base_l {parent=upperarm_twist_01_l, loc=<-0.3691 7.0506 -2.765>, rot=<0 0 0 1>}
 deform_upperarm_fr_l {parent=deform_upperarm_fr_base_l, loc=<0 0 0>, rot=<0 0 0 1>}
 deform_upperarm_up_base_l {parent=upperarm_twist_01_l, loc=<-3.2216 -1.1391 3.3311>, rot=<0 0 0 1>}
 deform_upperarm_up_l {parent=deform_upperarm_up_base_l, loc=<0 0 0>, rot=<0 0 0 1>}
 pad_l {parent=upperarm_twist_01_l, loc=<15.2358 -0.0889 6.6431>, rot=<0.0354 0.0674 -0.0059 -0.9971>}
 upperarm_twist_02_l {parent=upperarm_l, loc=<13.7941 0 0>, rot=<0 0 0 -1>}
 attach_cape {parent=spine_05, loc=<9.0206 -11.7146 -0.0002>, rot=<0.0616 -0.7044 -0.0616 -0.7044>}
 clavicle_r {parent=spine_05, loc=<4.3575 -0.1466 2.3223>, rot=<0.7289 0.0465 -0.6829 0.009>}
 upperarm_r {parent=clavicle_r, loc=<-15.5111 0 0>, rot=<0.0226 -0.3234 -0.0088 -0.9459>}
 lowerarm_r {parent=upperarm_r, loc=<-26.6514 0 0>, rot=<0.0391 -0.0099 -0.2453 -0.9686>}
 hand_r {parent=lowerarm_r, loc=<-22.3491 0 0>, rot=<-0.7081 -0.0507 0.0975 -0.6975>}
 thumb_01_r {parent=hand_r, loc=<-1.3118 0.3191 2.459>, rot=<0.6546 -0.4042 0.0664 -0.6354>}
 thumb_02_r {parent=thumb_01_r, loc=<-5.9336 0.0448 -0.0401>, rot=<-0.0037 0.0213 0.298 -0.9543>}
 thumb_03_r {parent=thumb_02_r, loc=<-2.2223 -0.3682 0.0907>, rot=<0.0118 -0.013 0.0066 -0.9998>}
 index_metacarpal_r {parent=hand_r, loc=<-1.6535 -0.0183 2.96>, rot=<0.0158 -0.0124 -0.0173 -0.9996>}
 index_01_r {parent=index_metacarpal_r, loc=<-8.3008 -0.0357 0.0524>, rot=<-0.0044 0.0485 0.0924 -0.9945>}
 index_02_r {parent=index_01_r, loc=<-3.3587 0.3004 -0.0452>, rot=<-0.002 0.0027 0.3225 -0.9466>}
 index_03_r {parent=index_02_r, loc=<-2.5578 0.123 0.0252>, rot=<-0.002 0.0067 0.2558 -0.9667>}
 middle_metacarpal_r {parent=hand_r, loc=<-2.1856 0.1725 0.5984>, rot=<0.006 0.0168 -0.0329 -0.9993>}
 middle_01_r {parent=middle_metacarpal_r, loc=<-7.3275 0.2447 -0.0129>, rot=<-0.0208 0.0042 0.0858 -0.9961>}
 middle_02_r {parent=middle_01_r, loc=<-4.0123 -0.1401 -0.1175>, rot=<-0.0218 0.0258 0.3512 -0.9357>}
 middle_03_r {parent=middle_02_r, loc=<-2.6077 -0.0534 -0.1373>, rot=<0.0082 0.0188 0.1708 -0.9851>}
 ring_metacarpal_r {parent=hand_r, loc=<-2.2074 0.3625 -1.3638>, rot=<-0.0284 0.0722 -0.0458 -0.9959>}
 ring_01_r {parent=ring_metacarpal_r, loc=<-7.0295 0.6197 0.1446>, rot=<-0.0489 -0.0252 0.1892 -0.9804>}
 ring_02_r {parent=ring_01_r, loc=<-3.6174 -0.4647 0.1027>, rot=<0 0 0.2806 -0.9598>}
 ring_03_r {parent=ring_02_r, loc=<-2.4152 -0.114 -0.0199>, rot=<0 0 0.1189 -0.9929>}
 pinky_metacarpal_r {parent=hand_r, loc=<-1.7616 0.9753 -3.3814>, rot=<-0.0397 0.0814 -0.069 -0.9935>}
 pinky_01_r {parent=pinky_metacarpal_r, loc=<-7.1709 0.4831 0.1094>, rot=<-0.0654 -0.0035 0.2242 -0.9723>}
 pinky_02_r {parent=pinky_01_r, loc=<-2.8274 -0.0893 0.1043>, rot=<0.0037 0.0011 0.2854 -0.9584>}
 pinky_03_r {parent=pinky_02_r, loc=<-2.2046 -0.0307 0.0078>, rot=<-0.0002 0.0098 0.1445 -0.9895>}
 deform_wrist_dn_base_r {parent=hand_r, loc=<-0.0004 2.2345 -0.0001>, rot=<0 0 0 -1>}
 deform_wrist_dn_r {parent=deform_wrist_dn_base_r, loc=<0 0 0>, rot=<0 0 0 1>}
 deform_wrist_up_base_r {parent=hand_r, loc=<0.0003 -2.2684 0.0001>, rot=<0 0 0 -1>}
 deform_wrist_up_r {parent=deform_wrist_up_base_r, loc=<0 0 0>, rot=<0 0 0 1>}
 lowerarm_twist_02_r {parent=lowerarm_r, loc=<-9.6199 0 0>, rot=<0 0 0 -1>}
 lowerarm_twist_01_r {parent=lowerarm_r, loc=<-17.8549 0 0>, rot=<0 0 0 -1>}
 deform_elbow_in_base_r {parent=lowerarm_r, loc=<-0.2111 -5.2591 -0.8275>, rot=<0 0 0 -1>}
 deform_elbow_in_r {parent=deform_elbow_in_base_r, loc=<-0.0001 0 0.0002>, rot=<0 0 0 1>}
 deform_elbow_out_base_r {parent=lowerarm_r, loc=<0 4 -0.0001>, rot=<0 0 0 -1>}
 deform_elbow_out_r {parent=deform_elbow_out_base_r, loc=<0 0 0>, rot=<0 0 0 1>}
 upperarm_twist_01_r {parent=upperarm_r, loc=<-0.0041 0 0>, rot=<0 0 0 -1>}
 deform_upperarm_bk_base_r {parent=upperarm_twist_01_r, loc=<-1.0814 7.2192 1.9455>, rot=<0 0 0 1>}
 deform_upperarm_bk_r {parent=deform_upperarm_bk_base_r, loc=<0 0 0>, rot=<0 0 0 1>}
 deform_upperarm_fr_base_r {parent=upperarm_twist_01_r, loc=<0.3689 -7.0506 2.7652>, rot=<0 0 0 1>}
 deform_upperarm_fr_r {parent=deform_upperarm_fr_base_r, loc=<0 0 0>, rot=<0 0 0 1>}
 deform_upperarm_up_base_r {parent=upperarm_twist_01_r, loc=<3.2214 1.1391 -3.3308>, rot=<0 0 0 1>}
 deform_upperarm_up_r {parent=deform_upperarm_up_base_r, loc=<0 0 0>, rot=<0 0 0 -1>}
 pad_r {parent=upperarm_twist_01_r, loc=<-15.2359 0.0889 -6.6429>, rot=<0.0354 0.0674 -0.0059 -0.9971>}
 upperarm_twist_02_r {parent=upperarm_r, loc=<-13.7941 0 0>, rot=<0 0 0 -1>}
 deform_pec_aim_base_l {parent=spine_04, loc=<7.8732 12.6388 -2.4123>, rot=<0.0102 -0.4602 0.3979 -0.7936>}
 deform_pec_aim_l {parent=deform_pec_aim_base_l, loc=<0 0 0>, rot=<0 0 0 1>}
 deform_pec_base_l {parent=deform_pec_aim_l, loc=<10.766 0 0>, rot=<0 0 0 1>}
 deform_pec_l {parent=deform_pec_base_l, loc=<0.055 0 0>, rot=<0 0 0 1>}
 deform_pec_aim_base_r {parent=spine_04, loc=<7.8727 12.6388 2.4123>, rot=<-0.4602 -0.0102 0.7936 0.3979>}
 deform_pec_aim_r {parent=deform_pec_aim_base_r, loc=<0 0 0>, rot=<0 0 0 1>}
 deform_pec_base_r {parent=deform_pec_aim_r, loc=<-10.7662 -0.0006 -0.0007>, rot=<0 0 0 1>}
 deform_pec_r {parent=deform_pec_base_r, loc=<0 0 0>, rot=<0 0 0 1>}
 thigh_l {parent=pelvis, loc=<-2.5693 -0.0109 -7.7501>, rot=<0.0012 0.0572 0.0155 -0.9982>}
 calf_l {parent=thigh_l, loc=<-39.0484 0 0>, rot=<0.0006 -0.0161 -0.0391 -0.9991>}
 foot_l {parent=calf_l, loc=<-42.561 0.1077 0.4292>, rot=<0.0418 -0.0455 0.0343 -0.9975>}
 ball_l {parent=foot_l, loc=<-6.1599 13.8264 0.6806>, rot=<0.0001 -0.0001 -0.7063 -0.7079>}
 ref_foot_l {parent=foot_l, loc=<0 0 0>, rot=<0 0 0 1>}
 calf_twist_02_l {parent=calf_l, loc=<-18.8064 -0.0967 0.0397>, rot=<0.0237 0.0095 -0.0035 -0.9997>}
 calf_twist_01_l {parent=calf_l, loc=<-35.813 -0.0967 0.0397>, rot=<0.0237 0.0095 -0.0035 -0.9997>}
 deform_knee_in_base_l {parent=calf_l, loc=<1.298 -5.2957 -0.0496>, rot=<0 0 0 -1>}
 deform_knee_in_l {parent=deform_knee_in_base_l, loc=<0 0 0>, rot=<0 0 0 1>}
 deform_knee_out_base_l {parent=calf_l, loc=<0 4 0>, rot=<0 0 0 -1>}
 deform_knee_out_l {parent=deform_knee_out_base_l, loc=<0 0 0>, rot=<0 0 0 1>}
 thigh_twist_01_l {parent=thigh_l, loc=<-21.6303 0 0>, rot=<0.0013 0 0 -1>}
 deform_groin_base_l {parent=thigh_l, loc=<-2.1908 8.1142 -1.7442>, rot=<0 0 0 -1>}
 deform_groin_l {parent=deform_groin_base_l, loc=<0 0 0>, rot=<0 0 0 1>}
 ref_thigh_l {parent=thigh_l, loc=<0 0 0>, rot=<0 0 0 1>}
 thigh_r {parent=pelvis, loc=<-2.5693 -0.0109 7.7501>, rot=<0.0572 -0.0012 0.9982 0.0155>}
 calf_r {parent=thigh_r, loc=<39.0484 0 0>, rot=<0.0006 -0.0161 -0.0391 -0.9991>}
 foot_r {parent=calf_r, loc=<42.561 -0.1077 -0.4292>, rot=<0.0418 -0.0455 0.0343 -0.9975>}
 ball_r {parent=foot_r, loc=<6.1599 -13.8264 -0.6806>, rot=<0.0001 -0.0001 -0.7063 -0.7079>}
 ref_foot_r {parent=foot_r, loc=<0 0 0>, rot=<0 0 0 1>}
 calf_twist_02_r {parent=calf_r, loc=<18.8064 0.0967 -0.0397>, rot=<0.0237 0.0095 -0.0035 -0.9997>}
 calf_twist_01_r {parent=calf_r, loc=<35.813 0.0967 -0.0397>, rot=<0.0237 0.0095 -0.0035 -0.9997>}
 deform_knee_in_base_r {parent=calf_r, loc=<-1.298 5.2957 0.0496>, rot=<0 0 0 -1>}
 deform_knee_in_r {parent=deform_knee_in_base_r, loc=<0 0 0>, rot=<0 0 0 1>}
 deform_knee_out_base_r {parent=calf_r, loc=<0 -4 0>, rot=<0 0 0 -1>}
 deform_knee_out_r {parent=deform_knee_out_base_r, loc=<0 0 0>, rot=<0 0 0 1>}
 thigh_twist_01_r {parent=thigh_r, loc=<21.6303 0 0>, rot=<0.0013 0 0 -1>}
 deform_groin_base_r {parent=thigh_r, loc=<2.1908 -8.1142 1.7442>, rot=<0 0 0 -1>}
 deform_groin_r {parent=deform_groin_base_r, loc=<0 0 0>, rot=<0 0 0 1>}
 ref_thigh_r {parent=thigh_r, loc=<0 0 0>, rot=<0 0 0 1>}
 deform_glute_base_l {parent=pelvis, loc=<-8.8064 -8.8222 -8.4287>, rot=<0.0275 0.0501 -0.4472 -0.8926>}
 deform_glute_l {parent=deform_glute_base_l, loc=<0 0 0>, rot=<0 0 0 1>}
 deform_glute_base_r {parent=pelvis, loc=<-8.8062 -8.8222 8.4287>, rot=<0.0501 -0.0275 0.8926 -0.4472>}
 deform_glute_r {parent=deform_glute_base_r, loc=<0 0 0>, rot=<0 0 0 1>}
 base_strap_A_r {parent=pelvis, loc=<6.9776 1.5853 14.6301>, rot=<0.1686 0.1732 0.9694 -0.0428>}
 dyn_strap_A_00_r {parent=base_strap_A_r, loc=<0 0 0>, rot=<0 0 0 1>}
 dyn_strap_A_01_r {parent=dyn_strap_A_00_r, loc=<7.0098 0 0>, rot=<0 -0.0132 0 -0.9999>}
 dyn_strap_A_02_r {parent=dyn_strap_A_01_r, loc=<9.1917 0 0>, rot=<0.0007 -0.0163 0.0415 -0.999>}
 dyn_strap_A_03_r {parent=dyn_strap_A_02_r, loc=<9.2515 0 0>, rot=<0 0 0 1>}
 base_strap_B_r {parent=pelvis, loc=<7.7705 -2.7435 14.8072>, rot=<0.171 -0.251 0.9522 0.0321>}
 dyn_strap_B_00_r {parent=base_strap_B_r, loc=<0 0 0>, rot=<0 0 0 1>}
 dyn_strap_B_01_r {parent=dyn_strap_B_00_r, loc=<6.7978 0 0>, rot=<0 0 -0.0169 -0.9999>}
 dyn_strap_B_02_r {parent=dyn_strap_B_01_r, loc=<9.2246 0 0>, rot=<0.0005 -0.0172 -0.0297 -0.9994>}
 dyn_strap_B_03_r {parent=dyn_strap_B_02_r, loc=<9.4244 0 0>, rot=<0 0 0 1>}
 base_frontCloth {parent=pelvis, loc=<8.0078 7.8517 0>, rot=<0 0 0.9968 0.0793>}
 base_frontCloth_l {parent=base_frontCloth, loc=<3.745 1.0391 -7.0107>, rot=<0.007 -0.0567 0.0008 -0.9984>}
 dyn_frontCloth_00_l {parent=base_frontCloth_l, loc=<0 0 0>, rot=<0 0 0 1>}
 dyn_frontCloth_01_l {parent=dyn_frontCloth_00_l, loc=<10.5364 -0.0126 0>, rot=<0 0 -0.0203 -0.9998>}
 dyn_frontCloth_02_l {parent=dyn_frontCloth_01_l, loc=<11.7976 0.0005 -0.0028>, rot=<0 0 0 1>}
 base_frontCloth_r {parent=base_frontCloth, loc=<3.745 1.0391 7.0107>, rot=<-0.007 0.0567 0.0008 -0.9984>}
 dyn_frontCloth_00_r {parent=base_frontCloth_r, loc=<0 0 0>, rot=<0 0 0 -1>}
 dyn_frontCloth_01_r {parent=dyn_frontCloth_00_r, loc=<10.5364 -0.0126 0>, rot=<0 0 -0.0203 -0.9998>}
 dyn_frontCloth_02_r {parent=dyn_frontCloth_01_r, loc=<11.7986 -0.0034 -0.0004>, rot=<0 0 0 1>}
 base_backCloth_l {parent=pelvis, loc=<4.8243 -8.9337 -6.863>, rot=<-0.0558 0.0124 0.9821 -0.1793>}
 dyn_backCloth_00_l {parent=base_backCloth_l, loc=<0 0 0>, rot=<0 0 0 1>}
 dyn_backCloth_01_l {parent=dyn_backCloth_00_l, loc=<10.5825 -0.5281 0>, rot=<0 0 0.128 -0.9918>}
 base_backCloth_r {parent=pelvis, loc=<4.8243 -8.9337 6.863>, rot=<0.0558 -0.0124 0.9821 -0.1793>}
 dyn_backCloth_00_r {parent=base_backCloth_r, loc=<0 0 0>, rot=<0 0 0 1>}
 dyn_backCloth_01_r {parent=dyn_backCloth_00_r, loc=<10.5825 -0.5282 0>, rot=<0 0 0.128 -0.9918>}
 base_frontMiddleCloth_00 {parent=pelvis, loc=<-8.944 9.786 0.061>, rot=<0 -0.0035 0.9992 0.0398>}
 dyn_frontMiddleCloth_00 {parent=base_frontMiddleCloth_00, loc=<0 0 0>, rot=<0 0 0 1>}
 dyn_frontMiddleCloth_01 {parent=dyn_frontMiddleCloth_00, loc=<9.4243 -0.0576 0>, rot=<0 0 -0.0203 -0.9998>}
 base_backMiddleCloth_00 {parent=pelvis, loc=<-7.701 -13.115 -0.099>, rot=<-0.0001 -0.0035 1 -0.0012>}
 dyn_backMiddleCloth_00 {parent=base_backMiddleCloth_00, loc=<0 0 0>, rot=<0 0 0 1>}
 dyn_backMiddleCloth_01 {parent=dyn_backMiddleCloth_00, loc=<8.2685 0.3503 0.001>, rot=<0 0 -0.0203 -0.9998>}
 ref_thigh_twist_01_l_ctrl {controlType=EULER_TRANSFORM, parent=pelvis, type=Control, loc=<0 0 0>, rot=<0 0 0 1>}
 ref_thigh_twist_01_r_ctrl {controlType=EULER_TRANSFORM, parent=pelvis, type=Control, loc=<0 0 0>, rot=<0 0 0 1>}
 thigh_r_null {parent=pelvis, type=Null, loc=<-2.5693 -0.0109 7.7501>, rot=<0.0572 -0.0012 0.9982 0.0155>}
 pelvis_ctrl {controlType=EULER_TRANSFORM, type=Control, loc=<0 0 0>, rot=<0 0 0 1>}
 spine_01_ctrl {controlType=EULER_TRANSFORM, parent=pelvis_ctrl, type=Control, loc=<0 0 0>, rot=<0 0 0 1>}
 spine_02_ctrl {controlType=EULER_TRANSFORM, parent=spine_01_ctrl, type=Control, loc=<0 0 0>, rot=<0 0 0 1>}
 spine_03_ctrl {controlType=EULER_TRANSFORM, parent=spine_02_ctrl, type=Control, loc=<0 0 0>, rot=<0 0 0 1>}
 spine_04_ctrl {controlType=EULER_TRANSFORM, parent=spine_03_ctrl, type=Control, loc=<0 0 0>, rot=<0 0 0 1>}
 thigh_l_ctrl {controlType=EULER_TRANSFORM, parent=pelvis_ctrl, type=Control, loc=<0 0 0>, rot=<0 0 0 1>}
 thigh_r_ctrl {controlType=EULER_TRANSFORM, parent=pelvis_ctrl, type=Control, loc=<0 0 0>, rot=<0 0 0 1>}
]
variables [
 GetCrouchingAlpha {type=double, default=0}
 Leg_l_Swing_rot {type=FVector, default=(X=0,Y=0,Z=0)}
 Leg_r_Swing_rot {type=FVector, default=(X=0,Y=0,Z=0)}
 SetLoinclothFront {type=FTransform, default=(Rotation=(X=0,Y=0,Z=0,W=1),Translation=(X=0,Y=0,Z=0),Scale3D=(X=1,Y=1,Z=1))}
 Sliding {type=bool, default=False}
 grindOnRail {type=bool, default=False}
 grindReduction {type=double, default=0}
 isRunningOrEmote {type=bool, default=False}
 l_thigh_back_alpha {type=double, default=0}
 l_thigh_up_alpha {type=double, default=0}
 leg_rot_y_add {type=double, default=0}
 leg_rot_z_minimum {type=double, default=0}
 r_thigh_back_alpha {type=double, default=0}
 r_thigh_up_alpha {type=double, default=0}
]
