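# Physics asset for the male medium TotalFlock body cosmetic, giving capsule bodies to its hanging front and back cloth and right-side straps.
# Cloth and strap chains hang from pelvis-attached base bones with tight angular limits, twist locked on straps and back cloth.
# A free constraint links the lower back panels.
# Spine, pelvis, leg and extra ref_ thigh and foot capsules, plus a huge root capsule below the origin, provide body volumes.
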PhysicsAsset /BRCosmetics/Characters/Player/Male/Medium/Bodies/M_MED_TotalFlock/Meshes/M_MED_TotalFlock_Physics
bodies {
 base_backCloth_l [{shapeType=Capsule, length=0.505, radius=0.505, shapeName=base_backCloth_l_capsule, rotation=<rot 0 -90 0>}]
 base_backCloth_r [{shapeType=Capsule, length=0.505, radius=0.505, shapeName=base_backCloth_r_capsule, rotation=<rot 0 -90 0>}]
 base_frontCloth [{shapeType=Capsule, length=0.505, radius=0.505, shapeName=base_frontCloth_capsule, rotation=<rot 0 -90 0>}]
 base_frontCloth_l [{shapeType=Capsule, length=0.505, radius=0.505, shapeName=base_frontCloth_l_capsule, rotation=<rot 0 -90 0>}]
 base_frontCloth_r [{shapeType=Capsule, length=0.505, radius=0.505, shapeName=base_frontCloth_r_capsule, rotation=<rot 0 -90 0>}]
 base_strap_A_r [{shapeType=Capsule, length=0.505, radius=0.505, shapeName=base_strap_A_r_capsule, rotation=<rot 0 -90 0>}]
 base_strap_B_r [{shapeType=Capsule, length=0.505, radius=0.505, shapeName=base_strap_B_r_capsule, rotation=<rot 0 -90 0>}]
 dyn_backCloth_00_l [
  {shapeType=Capsule, length=8.9223, radius=5.2547, shapeName=dyn_backCloth_00_l_capsule, center=<5 7 0>}
  {
   shapeType=Capsule
   length=3
   radius=6
   shapeName=dyn_backCloth_00_l_capsule
   center=<3.4203 -1.0695 -13.6107>
   rotation=<rot 6.4075 -8.3059 52.6028>
  }
 ]
 dyn_backCloth_00_r [
  {shapeType=Capsule, length=8.9223, radius=5.2547, shapeName=dyn_backCloth_00_r_capsule, center=<5 7 0>}
  {
   shapeType=Capsule
   length=3
   radius=6
   shapeName=dyn_backCloth_00_r_capsule
   center=<3.4203 -1.0695 13.6107>
   rotation=<rot 6.4075 -8.3059 -52.6028>
  }
 ]
 dyn_backCloth_01_l [
  {shapeType=Capsule, length=5.5, radius=5, shapeName=dyn_backCloth_01_l_capsule, center=<5 3.7717 0>}
  {shapeType=Capsule, length=1.4119, radius=5, shapeName=dyn_backCloth_01_l_capsule, center=<5.5583 -2.6974 -12.8015>, rotation=<rot 0 0 40>}
 ]
 dyn_backCloth_01_r [
  {shapeType=Capsule, length=5.5, radius=5, shapeName=dyn_backCloth_01_r_capsule, center=<3.871 4.7205 0>}
  {shapeType=Capsule, length=1.4119, radius=5, shapeName=dyn_backCloth_01_l_capsule, center=<5.5583 -2.6974 12.8015>, rotation=<rot 0 0 -40>}
 ]
 dyn_frontCloth_00_l [
  {shapeType=Capsule, length=7.4701, radius=5.4551, shapeName=dyn_frontCloth_00_l_capsule, center=<5.1626 -6.5 0>}
  {
   shapeType=Capsule
   length=1.6313
   radius=5.2494
   shapeName=dyn_frontCloth_00_l_capsule
   center=<5.1626 -1.4057 -8.9627>
   rotation=<rot 0 0 -53.5703>
  }
 ]
 dyn_frontCloth_00_r [
  {shapeType=Capsule, length=8.4319, radius=4.4933, shapeName=dyn_frontCloth_00_r_capsule, center=<5.1626 -6.5 0>}
  {
   shapeType=Capsule
   length=1.5905
   radius=4.9658
   shapeName=dyn_frontCloth_00_r_capsule
   center=<5.1626 -1.4057 8.9627>
   rotation=<rot 0 0 35>
  }
 ]
 dyn_frontCloth_01_l [
  {shapeType=Capsule, length=16.0154, radius=5.5807, shapeName=dyn_frontCloth_01_l_capsule, center=<5.3577 -4.3255 0>}
  {
   shapeType=Capsule
   length=3.8638
   radius=5.6948
   shapeName=dyn_frontCloth_01_l_capsule
   center=<6.3376 -1.7226 -8.7395>
   rotation=<rot 8.3604 -6.6143 -45.5519>
  }
 ]
 dyn_frontCloth_01_r [
  {shapeType=Capsule, length=15.8823, radius=5.7139, shapeName=dyn_frontCloth_01_l_capsule, center=<5.3577 -4.3255 0>}
  {
   shapeType=Capsule
   length=2.2679
   radius=5.1976
   shapeName=dyn_frontCloth_01_l_capsule
   center=<5.9939 -1.4719 8.9884>
   rotation=<rot 2.8832 2.4253 40.2659>
  }
 ]
 dyn_frontCloth_02_l [{
   shapeType=Capsule
   length=12.6195
   radius=5.2788
   shapeName=dyn_frontCloth_02_l_capsule
   center=<2.6647 -2.3112 0>
   rotation=<rot 0 0 -1.4744>
  }]
 dyn_frontCloth_02_r [{
   shapeType=Capsule
   length=12.6195
   radius=5.2788
   shapeName=dyn_frontCloth_02_r_capsule
   center=<2.6647 -2.3112 0>
   rotation=<rot 0 0 2.3215>
  }]
 dyn_frontMiddleCloth_01 [{
   shapeType=Capsule
   length=23.7882
   radius=8.6616
   shapeName=dyn_frontMiddleCloth_01_capsule
   center=<-1.2485 10.7818 0>
   rotation=<rot 0 -90 0>
  }]
 dyn_strap_A_00_r [{shapeType=Capsule, length=3.7481, radius=2.2193, shapeName=dyn_strap_A_00_r_capsule, center=<2.8668 0 0.746>, rotation=<rot 90 0 0>}]
 dyn_strap_A_01_r [{shapeType=Capsule, length=2.2154, radius=3.3439, shapeName=dyn_strap_A_01_r_capsule, center=<3.7324 -0.6192 1.6638>, rotation=<rot 90 0 0>}]
 dyn_strap_A_02_r [{shapeType=Capsule, length=1.4316, radius=4.0988, shapeName=dyn_strap_A_02_r_capsule, center=<3.6298 -1.21 1.4618>, rotation=<rot 90 0 0>}]
 dyn_strap_A_03_r [{shapeType=Capsule, length=6.9112, radius=4.086, shapeName=dyn_strap_A_03_r_capsule, center=<3.4431 -0.2769 0.646>, rotation=<rot 90 0 0>}]
 dyn_strap_B_00_r [{shapeType=Capsule, length=2.9513, radius=1.8276, shapeName=dyn_strap_B_00_r_capsule, center=<2.2772 0 0>, rotation=<rot 90 0 0>}]
 dyn_strap_B_01_r [{shapeType=Capsule, length=2.5899, radius=3.2318, shapeName=dyn_strap_B_01_r_capsule, center=<4.0809 -0.3895 1.8235>, rotation=<rot 90 0 0>}]
 dyn_strap_B_02_r [{shapeType=Capsule, length=1.4914, radius=4.1445, shapeName=dyn_strap_B_02_r_capsule, center=<4.2382 0.6295 1.4618>, rotation=<rot 90 0 0>}]
 dyn_strap_B_03_r [{shapeType=Capsule, length=7.6715, radius=3.9729, shapeName=dyn_strap_B_03_r_capsule, center=<3.9588 -0.2515 0.646>, rotation=<rot 90 0 0>}]
 foot_l [{shapeType=Capsule, length=18, radius=5, shapeName=foot_l_capsule, rotation=<rot 0 -90 -90>}]
 foot_r [{shapeType=Capsule, length=18, radius=5, shapeName=foot_r_capsule, rotation=<rot 0 -90 -90>}]
 pelvis [{shapeType=Capsule, length=8.6597, radius=11.2297, shapeName=pelvis_capsule, center=<2 -1.1855 0>, rotation=<rot 0 -90 0>}]
 ref_foot_l [{shapeType=Capsule, length=18, radius=15, shapeName=foot_l_capsule, rotation=<rot 0 -90 -90>}]
 ref_foot_r [{shapeType=Capsule, length=18, radius=15, shapeName=foot_r_capsule, rotation=<rot 0 -90 -90>}]
 ref_thigh_l [{shapeType=Capsule, length=38.0861, radius=10, shapeName=thigh_l_capsule, center=<-17.6578 -2.2911 0.2237>, rotation=<rot 86.4563 180 180>}]
 ref_thigh_r [{shapeType=Capsule, length=37.9471, radius=10.139, shapeName=thigh_r_capsule, center=<17.6578 2.2911 -0.2237>, rotation=<rot 86.4563 180 0>}]
 root [{shapeType=Capsule, length=2, radius=140, shapeName=root_capsule, center=<0 0 -127.8701>, rotation=<rot 0 -90 0>}]
 spine_01 [{shapeType=Capsule, length=0.505, radius=0.505, shapeName=spine_01_capsule, rotation=<rot 0 -90 0>}]
 spine_02 [{shapeType=Capsule, length=10, radius=12, shapeName=spine_02_capsule, center=<5.2537 0.2829 0.0762>, rotation=<rot -1.1933 -90 -0.1318>}]
 spine_03 [{
   shapeType=Capsule
   length=17.0806
   radius=13.332
   shapeName=spine_03_capsule
   center=<9.6643 0.1788 0.0247>
   rotation=<rot -0.139 -90 -0.1146>
  }]
 spine_04 [{shapeType=Capsule, length=9.1269, radius=3.8219, shapeName=spine_04_capsule, center=<7.0562 11.6196 0.014>, rotation=<rot 0.0958 -90 0.071>}]
 thigh_l [{shapeType=Capsule, length=35, radius=10, shapeName=thigh_l_capsule, center=<-14.8109 -2.2911 0.4>, rotation=<rot 86.4563 180 180>}]
 thigh_r [{shapeType=Capsule, length=35, radius=10, shapeName=thigh_l_capsule, center=<14.8109 2.2911 -0.4>, rotation=<rot 86.4563 180 180>}]
}
constraints [
 {bone1Name=dyn_strap_A_00_r, bone2Name=base_strap_A_r, swing1LimitDegrees=25, swing2LimitDegrees=23, twistLimitDegrees=20, twistMotion=Locked}
 {
  bone1Name=dyn_strap_A_01_r
  bone2Name=dyn_strap_A_00_r
  swing1LimitDegrees=25
  swing2LimitDegrees=23
  twistLimitDegrees=20
  twistMotion=Locked
 }
 {
  bone1Name=dyn_strap_A_02_r
  bone2Name=dyn_strap_A_01_r
  swing1LimitDegrees=25
  swing2LimitDegrees=23
  twistLimitDegrees=20
  twistMotion=Locked
 }
 {
  bone1Name=dyn_strap_A_03_r
  bone2Name=dyn_strap_A_02_r
  swing1LimitDegrees=25
  swing2LimitDegrees=23
  twistLimitDegrees=20
  twistMotion=Locked
 }
 {bone1Name=dyn_strap_B_00_r, bone2Name=base_strap_B_r, swing1LimitDegrees=10, swing2LimitDegrees=23, twistLimitDegrees=1, twistMotion=Locked}
 {
  bone1Name=dyn_strap_B_01_r
  bone2Name=dyn_strap_B_00_r
  swing1LimitDegrees=10
  swing2LimitDegrees=23
  twistLimitDegrees=1
  twistMotion=Locked
 }
 {
  bone1Name=dyn_strap_B_02_r
  bone2Name=dyn_strap_B_01_r
  swing1LimitDegrees=10
  swing2LimitDegrees=23
  twistLimitDegrees=1
  twistMotion=Locked
 }
 {
  bone1Name=dyn_strap_B_03_r
  bone2Name=dyn_strap_B_02_r
  swing1LimitDegrees=10
  swing2LimitDegrees=23
  twistLimitDegrees=1
  twistMotion=Locked
 }
 {
  bone1Name=base_strap_A_r
  bone2Name=pelvis
  swing1LimitDegrees=45
  swing1Motion=Free
  swing2LimitDegrees=45
  swing2Motion=Free
  twistLimitDegrees=45
  twistMotion=Free
 }
 {
  bone1Name=base_strap_B_r
  bone2Name=pelvis
  swing1LimitDegrees=45
  swing1Motion=Free
  swing2LimitDegrees=45
  swing2Motion=Free
  twistLimitDegrees=45
  twistMotion=Free
 }
 {
  bone1Name=thigh_r
  bone2Name=pelvis
  swing1LimitDegrees=45
  swing1Motion=Free
  swing2LimitDegrees=45
  swing2Motion=Free
  twistLimitDegrees=45
  twistMotion=Free
 }
 {
  bone1Name=base_backCloth_l
  bone2Name=pelvis
  swing1LimitDegrees=45
  swing1Motion=Free
  swing2LimitDegrees=45
  swing2Motion=Free
  twistLimitDegrees=45
  twistMotion=Free
 }
 {
  bone1Name=dyn_backCloth_00_l
  bone2Name=base_backCloth_l
  swing1LimitDegrees=25
  swing2LimitDegrees=25
  swing2Motion=Locked
  twistLimitDegrees=1
  twistMotion=Locked
 }
 {
  bone1Name=dyn_backCloth_01_l
  bone2Name=dyn_backCloth_00_l
  swing1LimitDegrees=25
  swing2LimitDegrees=25
  swing2Motion=Locked
  twistLimitDegrees=1
  twistMotion=Locked
 }
 {
  bone1Name=base_backCloth_r
  bone2Name=pelvis
  swing1LimitDegrees=45
  swing1Motion=Free
  swing2LimitDegrees=45
  swing2Motion=Free
  twistLimitDegrees=45
  twistMotion=Free
 }
 {
  bone1Name=dyn_backCloth_00_r
  bone2Name=base_backCloth_r
  swing1LimitDegrees=25
  swing2LimitDegrees=25
  swing2Motion=Locked
  twistLimitDegrees=1
  twistMotion=Locked
 }
 {
  bone1Name=dyn_backCloth_01_r
  bone2Name=dyn_backCloth_00_r
  swing1LimitDegrees=25
  swing2LimitDegrees=25
  swing2Motion=Locked
  twistLimitDegrees=1
  twistMotion=Locked
 }
 {
  bone1Name=base_frontCloth_r
  bone2Name=pelvis
  swing1LimitDegrees=45
  swing1Motion=Free
  swing2LimitDegrees=45
  swing2Motion=Free
  twistLimitDegrees=45
  twistMotion=Free
 }
 {bone1Name=dyn_frontCloth_00_r, bone2Name=base_frontCloth_r, swing1LimitDegrees=24, swing2LimitDegrees=4, twistLimitDegrees=12}
 {bone1Name=dyn_frontCloth_01_r, bone2Name=dyn_frontCloth_00_r, swing1LimitDegrees=24, swing2LimitDegrees=4, twistLimitDegrees=12}
 {
  bone1Name=base_frontCloth_l
  bone2Name=pelvis
  swing1LimitDegrees=45
  swing1Motion=Free
  swing2LimitDegrees=45
  swing2Motion=Free
  twistLimitDegrees=45
  twistMotion=Free
 }
 {bone1Name=dyn_frontCloth_00_l, bone2Name=base_frontCloth_l, swing1LimitDegrees=24, swing2LimitDegrees=4, twistLimitDegrees=12}
 {bone1Name=dyn_frontCloth_01_l, bone2Name=dyn_frontCloth_00_l, swing1LimitDegrees=24, swing2LimitDegrees=4, twistLimitDegrees=12}
 {
  bone1Name=base_frontCloth
  bone2Name=pelvis
  swing1LimitDegrees=45
  swing1Motion=Free
  swing2LimitDegrees=45
  swing2Motion=Free
  twistLimitDegrees=45
  twistMotion=Free
 }
 {
  bone1Name=thigh_l
  bone2Name=pelvis
  swing1LimitDegrees=45
  swing1Motion=Free
  swing2LimitDegrees=45
  swing2Motion=Free
  twistLimitDegrees=45
  twistMotion=Free
 }
 {
  bone1Name=dyn_backCloth_01_l
  bone2Name=dyn_backCloth_01_r
  swing1LimitDegrees=45
  swing1Motion=Free
  swing2LimitDegrees=45
  swing2Motion=Free
  twistLimitDegrees=45
  twistMotion=Free
 }
 {bone1Name=dyn_frontCloth_02_l, bone2Name=dyn_frontCloth_01_l, swing1LimitDegrees=24, swing2LimitDegrees=4, twistLimitDegrees=12}
 {bone1Name=dyn_frontCloth_02_r, bone2Name=dyn_frontCloth_01_r, swing1LimitDegrees=24, swing2LimitDegrees=4, twistLimitDegrees=12}
 {
  bone1Name=pelvis
  bone2Name=root
  swing1LimitDegrees=45
  swing1Motion=Free
  swing2LimitDegrees=45
  swing2Motion=Free
  twistLimitDegrees=45
  twistMotion=Free
 }
 {
  bone1Name=foot_l
  bone2Name=thigh_l
  swing1LimitDegrees=45
  swing1Motion=Free
  swing2LimitDegrees=45
  swing2Motion=Free
  twistLimitDegrees=45
  twistMotion=Free
 }
 {
  bone1Name=foot_r
  bone2Name=thigh_r
  swing1LimitDegrees=45
  swing1Motion=Free
  swing2LimitDegrees=45
  swing2Motion=Free
  twistLimitDegrees=45
  twistMotion=Free
 }
 {
  bone1Name=spine_01
  bone2Name=pelvis
  swing1LimitDegrees=45
  swing1Motion=Free
  swing2LimitDegrees=45
  swing2Motion=Free
  twistLimitDegrees=45
  twistMotion=Free
 }
 {
  bone1Name=spine_02
  bone2Name=spine_01
  swing1LimitDegrees=45
  swing1Motion=Free
  swing2LimitDegrees=45
  swing2Motion=Free
  twistLimitDegrees=45
  twistMotion=Free
 }
 {
  bone1Name=spine_03
  bone2Name=spine_02
  swing1LimitDegrees=45
  swing1Motion=Free
  swing2LimitDegrees=45
  swing2Motion=Free
  twistLimitDegrees=45
  twistMotion=Free
 }
 {
  bone1Name=spine_04
  bone2Name=spine_03
  swing1LimitDegrees=45
  swing1Motion=Free
  swing2LimitDegrees=45
  swing2Motion=Free
  twistLimitDegrees=45
  twistMotion=Free
 }
 {
  bone1Name=ref_foot_l
  bone2Name=foot_l
  swing1LimitDegrees=45
  swing1Motion=Free
  swing2LimitDegrees=45
  swing2Motion=Free
  twistLimitDegrees=45
  twistMotion=Free
 }
 {
  bone1Name=ref_foot_r
  bone2Name=foot_r
  swing1LimitDegrees=45
  swing1Motion=Free
  swing2LimitDegrees=45
  swing2Motion=Free
  twistLimitDegrees=45
  twistMotion=Free
 }
 {
  bone1Name=ref_thigh_r
  bone2Name=thigh_r
  swing1LimitDegrees=45
  swing1Motion=Free
  swing2LimitDegrees=45
  swing2Motion=Free
  twistLimitDegrees=45
  twistMotion=Free
 }
 {
  bone1Name=ref_thigh_l
  bone2Name=thigh_l
  swing1LimitDegrees=45
  swing1Motion=Free
  swing2LimitDegrees=45
  swing2Motion=Free
  twistLimitDegrees=45
  twistMotion=Free
 }
 {
  bone1Name=dyn_frontMiddleCloth_01
  bone2Name=pelvis
  swing1LimitDegrees=45
  swing1Motion=Free
  swing2LimitDegrees=45
  swing2Motion=Free
  twistLimitDegrees=45
  twistMotion=Free
 }
]
