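# Animation Blueprint for the face accessory mesh of the TotalFlock male medium player body, following the parent mesh's pose.
# A copied parent pose feeds a Control Rig driven by crouch, crouch-walking and targeting states.
# A Blend Poses by bool node sits between the plain copied pose and the rigged result before output.
# Each update sets Crouch Walking as a float from moving and crouching both being true.
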AnimBlueprint /BRCosmetics/Characters/Player/Male/Medium/Bodies/M_MED_TotalFlock/Meshes/Parts/M_MED_TotalFlock_FaceAcc_AnimBP
target_skeleton=&/M_MED_TotalFlock_FaceAcc_Skeleton
graphs [
 AnimGraph {nodes [
   n1 = Control Rig!(
    Source=Copy Pose From Mesh() {bUseAttachedParent=true}
    Alpha
    Is Crouch Walking:double=Get Crouch Walking() -> (Crouch Walking:double)
    Is Crouching:bool=Get bIsCrouching()
    Is Targeting:bool=Get bIsTargeting()
   ) -> (Pose) {
    alphaBoolBlend {blendOption=Linear}
    controlRigAssetReference {blueprintRigClass=&/M_MED_TotalFlock_FaceAcc_CtrlRig.M_MED_TotalFlock_FaceAcc_CtrlRig_C}
   }
   Output Pose(Result=Blend Poses by bool(BlendPose_0=Copy Pose From Mesh() {bUseAttachedParent=true}, BlendPose_1=@n1, BlendTime_0=1, BlendTime_1=10) {blendType=Linear}) {layerGroup=None}
  ]}
 EventGraph {nodes [
   Event BlueprintUpdateAnimation()
   n1 = Get bIsMoving2D()
   n2 = Get bIsCrouching()
   > Set Crouch Walking(Crouch Walking:double="To Float (Boolean)"(InBool=AND Boolean(A=@n1, B=@n2))) -> (Output_Get:double)
  ]}
]
variables [
 Crouch Walking {type=number}
 Playing Emote {type=number}
]
